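# Niagara confetti effect for the hand in the GarlicWhisk Clobber emote, with two multicolored sprite emitters.
# Confetti fires two instantaneous bursts with spawn-group-scaled velocity, while Confetti_MovingHand spawns probabilistically per unit moved to leave a trail.
# Both share drag, gravity, curl noise and bouncy collision gated by quality level or front-end preview.
# Three sprite renderers use per-sprite sizes and colors, including time-of-day tinting, and particles shrink and die as they slow.
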NiagaraSystem /BRCosmetics/Animation/Game/MainPlayer/Emotes/GarlicWhisk_Clobber/FX/NS_GarlicWhisk_ClobberHandConfettiBurst
dependencies {
 usedDataInterfaces [
  &NiagaraDataInterfaceArrayColor
  &NiagaraDataInterfaceArrayFloat
  &NiagaraDataInterfacePlatformSet
  &NiagaraDataInterfaceCurve
 ]
 usedDynamicInputs [&SimulationPosition, &RotateVectorByQuaternion]
 usedModules [
  &SystemState
  &self:SystemSpawnScript.SystemScriptSource.SystemScriptGraph.NiagaraNodeAssignment_1.SetVariables_#1
 ]
}
emitters {
 Confetti {
  emitterUpdateScript [
   &EmitterState()
   &SpawnBurst_Instantaneous(Spawn Count=18, Use Spawn Probability=false)
   &SpawnBurst_Instantaneous(Spawn Count=4, Spawn Group=2, Use Spawn Probability=false) {module=SpawnBurst_Instantaneous002}
  ]
  particleSpawnScript [
   /SetVariables(Particles.SpawnGroup=$Emitter.SpawnGroup) {isSetParameters=true, module=SetVariables_#2}
   /SetVariables(Particles.VisibilityTag=$Particles.SpawnGroup) {isSetParameters=true, module=SetVariables_#3}
   &InitializeParticle(
    Color=&SelectLinearColorFromArray(
     Array Sampling Mode=::NewEnumerator0
     Direct Array Index=?
     Color Selection Array=DataInterface() {propertyValues {ColorData [<rgba 1 0.1614 0.9113 1>, <rgba 0.148 0 1 1>, <rgba 0.0772 0.9637 1 1>, <rgba 1 0.9088 0.0365 1>, <rgba 0.9385 0 0 0>]}}
     Random Seed=0
    )
    Lifetime=&RandomRangeFloat(Minimum=3, Maximum=4)
    Sprite Size Max=<3 5>
    Sprite Size Min=<2 4>
    Sprite Size Mode=::NewEnumerator2
    UsePositionOffset=false
    Write Material Random=true
   ) {module=InitializeParticle001}
   /SetVariables(
    Particles.Sprite_01_Color:LinearColor=&MakeCustomLinearColorFromBool(
     Bool=&RandomBool()
     True Color=&RandomRangeLinearColor(Minimum=<rgba 0 0.305 0 1>, Maximum=<rgba 0.4 0.305 0 1>)
     False Color=&RandomRangeLinearColor(Minimum=<rgba 0.85 0.305 0 1>, Maximum=<rgba 1 0.305 0 1>)
    )
    Particles.Sprite_01_Size:Vector2f=&MakeVector2D(
     X=&MakeFloatFromVector2d(Vector 2D=&Vector2DFromFloat(Value=&RandomRangeFloat(Minimum=1, Maximum=2)))
     Y=&MakeFloatFromVector2d(
      Vector 2D=&Vector2DFromFloat(Value=&RandomRangeFloat(Minimum=2, Maximum=3))
      Channel=::NewEnumerator1
     )
    )
   ) {isSetParameters=true, module=SetVariables_#4}
   /SetVariables(
    Particles.Sprite_02_Color:LinearColor=&HueShiftLinearColor(
     Hue=0
     Saturation=&RandomRangeFloat(Minimum=2, Maximum=8)
     Value=&RandomRangeFloat(Minimum=1, Maximum=10)
     Linear Color=&RandomRangeLinearColor(Minimum=<rgba 0 0 0 1>, Maximum=<rgba 1 1 1 1>, Color Channel Mode=::NewEnumerator2)
    )
    Particles.Sprite_02_Size:Vector2f=&MakeVector2D(
     X=&MakeFloatFromVector2d(Vector 2D=&Vector2DFromFloat(Value=&RandomRangeFloat(Minimum=10, Maximum=10)))
     Y=&MakeFloatFromVector2d(
      Vector 2D=&Vector2DFromFloat(Value=&RandomRangeFloat(Minimum=2, Maximum=3))
      Channel=::NewEnumerator1
     )
    )
   ) {isSetParameters=true, module=SetVariables_#5}
   &SphereLocation(
    MaskSpawnGroup=false
    Override Seed=false
    Sphere Radius=2
    Surface Expansion Mode=::NewEnumerator2
    Surface Only=true
    UseNonUniformScale=false
    UseOffset=true
    UseSphereOrientation=false
   )
   &AddVelocity(
    Module.Use Falloff Distance=false
    Origin Offset=<0 0 -1>
    Override Seed=false
    Velocity Mode=::NewEnumerator1
    Velocity Speed=/Niagara/DynamicInputs/Math/LerpMultipleFloats(
     Float Count:ENiagara_MultipleLerpCount=::NewEnumerator0
     1:NiagaraFloat=&RandomRangeFloat(Minimum=50, Maximum=112.5)
     2:NiagaraFloat=&RandomRangeFloat(Minimum=62.5, Maximum=162.5)
     3:NiagaraFloat=&RandomRangeFloat(Minimum=50, Maximum=150)
     4:NiagaraFloat=?
     5:NiagaraFloat=?
     Alpha:NiagaraFloat=&Divide_Float(
      A=&MakeFloatFromInt(Int=3)
      B=&MakeFloatFromInt(Int=$Particles.SpawnGroup)
     )
    )
   ) {module=AddVelocity001}
   /SetVariables(Particles.Color=&MakeLinearColorFromVectorAndFloat(
     "Vector (RGB)"=&AddVectorToPosition(
      Position=HlslExpression() {hlslExpression=Particles.Color.xyz}
      Vector=&FN_FrontEndVector3Modifier(Gameplay=<0 0 0>, Front End=<0 0.1 0.25>)
     )
     "Float (Alpha)"=1
    )) {isSetParameters=true, module=SetVariables_#6}
   &CurlNoiseForce(
    Bias Noise=false
    Noise Strength=&RandomRangeFloat(Minimum=25, Maximum=50)
    Pan Noise=true
    Pan Noise Field=<0.2016 0 0.5008>
   ) {module=CurlNoiseForce001}
   &AddVelocity(Velocity=<-10 10 -10>, Velocity Speed Scale=&RandomRangeFloat(Minimum=10, Maximum=15))
   &InheritVelocity(
    Inherited Velocity Amount Scale=<1 0 0>
    Inherited Velocity Speed Limit=400
    Speed Limit=true
    UseSpeedThreshold=false
    Velocity Destination Space:ENiagaraCoordinateSpace=::World
   )
   &ApplyInitialForces()
  ]
  particleUpdateScript [
   &ParticleState()
   &SpriteRotationRate(Rotation Rate=&Multiply_Float(
     A=&SelectFloatFromArray(
      Array Sampling Mode=::NewEnumerator0
      Direct Array Index=?
      Float Selection Array=DataInterface() {propertyValues {FloatData [-300,300]}}
      Random Seed=0
     )
     B=&RandomRangeFloat(Minimum=0.5, Maximum=1)
    ))
   &Drag(Drag=&RandomRangeFloat(Minimum=5, Maximum=7), Use Linear Drag=true, Use Rotational Drag=false)
   &GravityForce(Gravity=&RandomRangeVector(Minimum=<0 0 -350>, Maximum=<1 1 -225>))
   &CurlNoiseForce(Noise Strength=&RandomRangeFloat(Minimum=200, Maximum=300), Pan Noise=false)
   &Collision(
    Age Colliding Particles=false
    CPU Trace Vector Length Multiplier=1
    Collision Enabled=&Boolean_OrOperation(
     Bool1=&IsPlatformSetActive(New Platform Set=DataInterface() {propertyValues {Platforms {qualityLevelMask=-4}}})
     Bool2=$User.bIsFrontEndPreview
    )
    Enable Rest State=false
    Friction=0.5
    Module.EnableMaxCollisionCount=false
    Module.Randomize Collision Normal Vector=false
    Particle Radius Scale=0
    Restitution=0.9
    Restitution Coefficient Blending:ENiagaraRestitutionMergeType=::NewEnumerator1
    Simple Friction=false
   )
   &SolveForcesAndVelocity(Clamp Velocity=false, Limit Acceleration=false)
   &ScaleSpriteSizeBySpeed(Max Scale Factor=<1 1>, Min Scale Factor=<0 0>, Velocity Threshold=20)
   &ScaleSpriteSize(Scale Factor=&Vector2DFromFloat(Value=&FloatFromCurve(
      FloatCurve=DataInterface() {propertyValues {Curve {keys [
          {arriveTangent=-0.4596, interpMode=RCIM_Cubic, leaveTangent=-0.4596, tangentMode=RCTM_User, time=0.55, value=1}
          {arriveTangent=-4.9031, interpMode=RCIM_Cubic, leaveTangent=-4.9031, tangentMode=RCTM_User, time=1, value=0}
         ]}}}
      Scale Curve=&RandomRangeFloat(Minimum=0.4, Maximum=1, Randomness Mode=::NewEnumerator2, Random Seed=?)
     )))
   /SetVariables(
    Particles.Sprite_01_Color:LinearColor=&MultiplyLinearColor(Linear Color=$Particles.Initial.Sprite_01_Color, Scale Factor=<1 1 1 1>)
    Particles.Sprite_01_Size:Vector2f=&Multiply_Vector2D(
     A=$Particles.Initial.Sprite_01_Size
     B=&Vector2DFromFloat(Value=&FloatFromCurve(
       FloatCurve=DataInterface() {propertyValues {Curve {keys [
           {arriveTangent=-0.4596, interpMode=RCIM_Cubic, leaveTangent=-0.4596, tangentMode=RCTM_User, time=0.55, value=1}
           {arriveTangent=-4.9031, interpMode=RCIM_Cubic, leaveTangent=-4.9031, tangentMode=RCTM_User, time=1, value=0}
          ]}}}
       CurveIndex=&Divide_Float(
        A=$Particles.NormalizedAge
        B=&RandomRangeFloat(Minimum=0.5, Maximum=1, Randomness Mode=::NewEnumerator2, Random Seed=?)
       )
      ))
    )
   ) {isSetParameters=true, module=SetVariables_#7}
   /SetVariables(
    Particles.Sprite_02_Color:LinearColor=&MultiplyLinearColor(
     Linear Color=$Particles.Initial.Sprite_02_Color
     Scale Factor=&MakeVector4_FromLinearColor(LinearColor=&MultiplyLinearColor(
       Linear Color=&FN_Select_Color_from_Time_Of_Day(
        Use Manual Time of Day Thresholds=true
        Morning=<rgba 1 0.75 0.7753 1>
        Day=<rgba 1 0.8842 0.8294 1>
        Evening=<rgba 1 0.625 0.5 1>
        Night=<rgba 0.7 0.3494 0.4483 1>
        Evening Threshold=21
       )
       Scale Factor=&MakeFloat4FromVectorAndFloat(
        "Vector (RGB)"=<1 1 1>
        "Float (Alpha)"=&Lerp_Float(
         A=1
         B=0
         Alpha=&Divide_Float(
          A=$Particles.NormalizedAge
          B=&RandomRangeFloat(Minimum=0.2, Maximum=1, Randomness Mode=::NewEnumerator2, Random Seed=?)
         )
        )
       )
      ))
    )
    Particles.Sprite_02_Size:Vector2f=&Multiply_Vector2D(
     A=$Particles.Initial.Sprite_02_Size
     B=&Vector2DFromFloat(Value=&FloatFromCurve(
       FloatCurve=DataInterface() {propertyValues {Curve {keys [<at 0.0018 0.1957>,<at 0.0527 0.819>,<at 0.1169 0.6555>,<at 0.1748 0.9015>,<at 0.2394 0.582>,<at 0.332 1.0023>,<at 0.4947 0.3707>,<at 0.6727 1>,<at 1 0>]}}}
       CurveIndex=&Divide_Float(
        A=$Particles.NormalizedAge
        B=&RandomRangeFloat(Minimum=0.7, Maximum=1, Randomness Mode=::NewEnumerator2, Random Seed=?)
       )
      ))
    )
   ) {isSetParameters=true, module=SetVariables_#8}
   &KillParticles(Kill Particles=&SetBoolByFloatComparison(A=&VectorLength(Vector=$Particles.Velocity), B=1, Comparison Type=::NewEnumerator6))
  ]
  rendererClasses [&NiagaraSpriteRendererProperties]
  renderers [
   {rendererClass=&NiagaraSpriteRendererProperties}
   {rendererClass=&NiagaraSpriteRendererProperties, rendererIndex=1}
   {rendererClass=&NiagaraSpriteRendererProperties, rendererIndex=2}
  ]
 }
 Confetti_MovingHand {
  emitterUpdateScript [
   &EmitterState(Life Cycle Mode=::NewEnumerator1, Loop Behavior=::NewEnumerator1, Loop Delay=2, Loop Duration=1, UseLoopDelay=false)
   &SpawnPerUnit(
    Max Movement Threshold=300
    Module.Use Max Movement Threshold=true
    Module.Use Movement Tolerance=false
    Spawn Probability=0.7
    Spawn Spacing=1.7
    Use Spawn Probability=true
   )
  ]
  particleSpawnScript [
   /SetVariables(Particles.SpawnGroup=$Emitter.SpawnGroup) {isSetParameters=true, module=SetVariables_#2}
   /SetVariables(Particles.VisibilityTag=$Particles.SpawnGroup) {isSetParameters=true, module=SetVariables_#3}
   &InitializeParticle(
    Color=&SelectLinearColorFromArray(
     Array Sampling Mode=::NewEnumerator0
     Direct Array Index=?
     Color Selection Array=DataInterface() {propertyValues {ColorData [<rgba 1 0.1614 0.9113 1>, <rgba 0.148 0 1 1>, <rgba 0.0772 0.9637 1 1>, <rgba 1 0.9088 0.0365 1>, <rgba 0.9385 0 0 0>]}}
     Random Seed=0
    )
    Lifetime=&RandomRangeFloat(Minimum=3, Maximum=4)
    Sprite Size Max=<3 5>
    Sprite Size Min=<2 4>
    Sprite Size Mode=::NewEnumerator2
    UsePositionOffset=false
    Write Material Random=true
   ) {module=InitializeParticle001}
   /SetVariables(
    Particles.Sprite_01_Color:LinearColor=&MakeCustomLinearColorFromBool(
     Bool=&RandomBool()
     True Color=&RandomRangeLinearColor(Minimum=<rgba 0 0.305 0 1>, Maximum=<rgba 0.4 0.305 0 1>)
     False Color=&RandomRangeLinearColor(Minimum=<rgba 0.85 0.305 0 1>, Maximum=<rgba 1 0.305 0 1>)
    )
    Particles.Sprite_01_Size:Vector2f=&MakeVector2D(
     X=&MakeFloatFromVector2d(Vector 2D=&Vector2DFromFloat(Value=&RandomRangeFloat(Minimum=1, Maximum=2)))
     Y=&MakeFloatFromVector2d(
      Vector 2D=&Vector2DFromFloat(Value=&RandomRangeFloat(Minimum=2, Maximum=3))
      Channel=::NewEnumerator1
     )
    )
   ) {isSetParameters=true, module=SetVariables_#4}
   /SetVariables(
    Particles.Sprite_02_Color:LinearColor=&HueShiftLinearColor(
     Hue=0
     Saturation=&RandomRangeFloat(Minimum=2, Maximum=8)
     Value=&RandomRangeFloat(Minimum=1, Maximum=10)
     Linear Color=&RandomRangeLinearColor(Minimum=<rgba 0 0 0 1>, Maximum=<rgba 1 1 1 1>, Color Channel Mode=::NewEnumerator2)
    )
    Particles.Sprite_02_Size:Vector2f=&MakeVector2D(
     X=&MakeFloatFromVector2d(Vector 2D=&Vector2DFromFloat(Value=&RandomRangeFloat(Minimum=10, Maximum=10)))
     Y=&MakeFloatFromVector2d(
      Vector 2D=&Vector2DFromFloat(Value=&RandomRangeFloat(Minimum=2, Maximum=3))
      Channel=::NewEnumerator1
     )
    )
   ) {isSetParameters=true, module=SetVariables_#5}
   &SphereLocation(
    MaskSpawnGroup=false
    Override Seed=false
    Sphere Radius=2
    Surface Expansion Mode=::NewEnumerator2
    Surface Only=true
    UseNonUniformScale=false
    UseOffset=true
    UseSphereOrientation=false
   )
   /SetVariables(Particles.Color=&MakeLinearColorFromVectorAndFloat(
     "Vector (RGB)"=&AddVectorToPosition(
      Position=HlslExpression() {hlslExpression=Particles.Color.xyz}
      Vector=&FN_FrontEndVector3Modifier(Gameplay=<0 0 0>, Front End=<0 0.1 0.25>)
     )
     "Float (Alpha)"=1
    )) {isSetParameters=true, module=SetVariables_#6}
   &ApplyInitialForces()
  ]
  particleUpdateScript [
   &ParticleState()
   &SpriteRotationRate(Rotation Rate=&Multiply_Float(
     A=&SelectFloatFromArray(
      Array Sampling Mode=::NewEnumerator0
      Direct Array Index=?
      Float Selection Array=DataInterface() {propertyValues {FloatData [-300,300]}}
      Random Seed=0
     )
     B=&RandomRangeFloat(Minimum=0.5, Maximum=1)
    ))
   &Drag(Drag=&RandomRangeFloat(Minimum=5, Maximum=7), Use Linear Drag=true, Use Rotational Drag=false)
   &GravityForce(Gravity=&RandomRangeVector(Minimum=<0 0 -350>, Maximum=<1 1 -225>))
   &CurlNoiseForce(Noise Strength=&RandomRangeFloat(Minimum=200, Maximum=300), Pan Noise=false)
   &Collision(
    Age Colliding Particles=false
    CPU Trace Vector Length Multiplier=1
    Collision Enabled=&Boolean_OrOperation(
     Bool1=&IsPlatformSetActive(New Platform Set=DataInterface() {propertyValues {Platforms {qualityLevelMask=-4}}})
     Bool2=$User.bIsFrontEndPreview
    )
    Enable Rest State=false
    Friction=0.5
    Module.EnableMaxCollisionCount=false
    Module.Randomize Collision Normal Vector=false
    Particle Radius Scale=0
    Restitution=0.9
    Restitution Coefficient Blending:ENiagaraRestitutionMergeType=::NewEnumerator1
    Simple Friction=false
   )
   &SolveForcesAndVelocity(Clamp Velocity=false, Limit Acceleration=false)
   &ScaleSpriteSizeBySpeed(Max Scale Factor=<1 1>, Min Scale Factor=<0 0>, Velocity Threshold=20)
   &ScaleSpriteSize(Scale Factor=&Vector2DFromFloat(Value=&FloatFromCurve(
      FloatCurve=DataInterface() {propertyValues {Curve {keys [
          {arriveTangent=-0.4596, interpMode=RCIM_Cubic, leaveTangent=-0.4596, tangentMode=RCTM_User, time=0.55, value=1}
          {arriveTangent=-4.9031, interpMode=RCIM_Cubic, leaveTangent=-4.9031, tangentMode=RCTM_User, time=1, value=0}
         ]}}}
      Scale Curve=&RandomRangeFloat(Minimum=0.4, Maximum=1, Randomness Mode=::NewEnumerator2, Random Seed=?)
     )))
   /SetVariables(
    Particles.Sprite_01_Color:LinearColor=&MultiplyLinearColor(Linear Color=$Particles.Initial.Sprite_01_Color, Scale Factor=<1 1 1 1>)
    Particles.Sprite_01_Size:Vector2f=&Multiply_Vector2D(
     A=$Particles.Initial.Sprite_01_Size
     B=&Vector2DFromFloat(Value=&FloatFromCurve(
       FloatCurve=DataInterface() {propertyValues {Curve {keys [
           {arriveTangent=-0.4596, interpMode=RCIM_Cubic, leaveTangent=-0.4596, tangentMode=RCTM_User, time=0.55, value=1}
           {arriveTangent=-4.9031, interpMode=RCIM_Cubic, leaveTangent=-4.9031, tangentMode=RCTM_User, time=1, value=0}
          ]}}}
       CurveIndex=&Divide_Float(
        A=$Particles.NormalizedAge
        B=&RandomRangeFloat(Minimum=0.5, Maximum=1, Randomness Mode=::NewEnumerator2, Random Seed=?)
       )
      ))
    )
   ) {isSetParameters=true, module=SetVariables_#7}
   /SetVariables(
    Particles.Sprite_02_Color:LinearColor=&MultiplyLinearColor(
     Linear Color=$Particles.Initial.Sprite_02_Color
     Scale Factor=&MakeVector4_FromLinearColor(LinearColor=&MultiplyLinearColor(
       Linear Color=&FN_Select_Color_from_Time_Of_Day(
        Use Manual Time of Day Thresholds=true
        Morning=<rgba 1 0.75 0.7753 1>
        Day=<rgba 1 0.8842 0.8294 1>
        Evening=<rgba 1 0.625 0.5 1>
        Night=<rgba 0.7 0.3494 0.4483 1>
        Evening Threshold=21
       )
       Scale Factor=&MakeFloat4FromVectorAndFloat(
        "Vector (RGB)"=<1 1 1>
        "Float (Alpha)"=&Lerp_Float(
         A=1
         B=0
         Alpha=&Divide_Float(
          A=$Particles.NormalizedAge
          B=&RandomRangeFloat(Minimum=0.2, Maximum=1, Randomness Mode=::NewEnumerator2, Random Seed=?)
         )
        )
       )
      ))
    )
    Particles.Sprite_02_Size:Vector2f=&Multiply_Vector2D(
     A=$Particles.Initial.Sprite_02_Size
     B=&Vector2DFromFloat(Value=&FloatFromCurve(
       FloatCurve=DataInterface() {propertyValues {Curve {keys [<at 0.0018 0.1957>,<at 0.0527 0.819>,<at 0.1169 0.6555>,<at 0.1748 0.9015>,<at 0.2394 0.582>,<at 0.332 1.0023>,<at 0.4947 0.3707>,<at 0.6727 1>,<at 1 0>]}}}
       CurveIndex=&Divide_Float(
        A=$Particles.NormalizedAge
        B=&RandomRangeFloat(Minimum=0.7, Maximum=1, Randomness Mode=::NewEnumerator2, Random Seed=?)
       )
      ))
    )
   ) {isSetParameters=true, module=SetVariables_#8}
   &KillParticles(Kill Particles=&SetBoolByFloatComparison(A=&VectorLength(Vector=$Particles.Velocity), B=1, Comparison Type=::NewEnumerator6))
  ]
  rendererClasses [&NiagaraSpriteRendererProperties]
  renderers [
   {rendererClass=&NiagaraSpriteRendererProperties}
   {rendererClass=&NiagaraSpriteRendererProperties, rendererIndex=1}
   {rendererClass=&NiagaraSpriteRendererProperties, rendererIndex=2}
  ]
 }
}
summary {
 systemName=NS_GarlicWhisk_ClobberHandConfettiBurst
 emitters [
  {emitterName=Confetti, rendererClasses [&NiagaraSpriteRendererProperties]}
  {emitterName=Confetti_MovingHand, rendererClasses [&NiagaraSpriteRendererProperties]}
 ]
 userVariables [
  User.bIsFrontEnd {type=NiagaraBool}
  User.bIsFrontEndPreview {type=NiagaraBool}
 ]
}
