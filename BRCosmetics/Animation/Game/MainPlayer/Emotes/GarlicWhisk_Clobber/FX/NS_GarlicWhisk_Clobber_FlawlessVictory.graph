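# Niagara system for the GarlicWhisk Clobber emote that shows a fading Flawless Victory text sprite with flame, ember and flare effects.
# A brief flare accompanies timed GPU spawn embers, whose emitter also triggers flame audio.
# A second GPU ember emitter projects particles onto scene depth, culls them by GBuffer query, then drifts them with curl noise.
# User.bIsFrontEndPreview lowers the text and both ember emitters for front-end preview.
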NiagaraSystem /BRCosmetics/Animation/Game/MainPlayer/Emotes/GarlicWhisk_Clobber/FX/NS_GarlicWhisk_Clobber_FlawlessVictory
dependencies {
 usedDataInterfaces [
  &NiagaraDataInterfaceCurve
  &NiagaraDataInterfaceCamera
  &NiagaraDataInterfaceGBuffer
  &NiagaraDataInterfaceArrayFloat
  &NiagaraDataInterfaceAudioPlayer
 ]
 usedDynamicInputs [
  &SimulationPosition
  &ReturnCameraVector
  &RotateVectorByQuaternion
  &self:ScratchDynamicInput
 ]
 usedModules [
  &SystemState
  &self:SystemSpawnScript.SystemScriptSource.SystemScriptGraph.NiagaraNodeAssignment_1.SetVariables_#1
 ]
}
emitters {
 Ember {
  simTarget=GPUComputeSim
  emitterUpdateScript [
   &EmitterState()
   &SpawnRate(
    SpawnRate=&FloatFromCurve(
     FloatCurve=DataInterface() {propertyValues {Curve {keys [<at 0 1 RCIM_Constant>,<at 0.05 0.4>]}}}
     CurveIndex=$Emitter.NormalizedLoopAge
     Scale Curve=150
    )
    Use Spawn Probability=false
   )
   &SpawnRate(
    SpawnGroup=1
    SpawnRate=&FloatFromCurve(
     FloatCurve=DataInterface() {propertyValues {Curve {keys [<at 0 1 RCIM_Constant>,<at 0.05 0.4>]}}}
     CurveIndex=$Emitter.NormalizedLoopAge
     Scale Curve=100
    )
    Use Spawn Probability=false
   ) {module=SpawnRate001}
   &SpawnBurst_Instantaneous(Spawn Count=500, Spawn Time=0.1, Use Spawn Probability=true)
   &SpawnBurst_Instantaneous(Spawn Count=3000, Spawn Group=2, Spawn Time=0.05, Use Spawn Probability=true) {module=SpawnBurst_Instantaneous002}
  ]
  particleSpawnScript [
   /SetVariables(Particles.SpawnGroup=$Emitter.SpawnGroup) {isSetParameters=true}
   /SetVariables(Particles.VisibilityTag=$Particles.SpawnGroup) {isSetParameters=true}
   &InitializeParticle(
    Color=<rgba 60 8.0591 0 1>
    Lifetime=&RandomRangeFloat(Minimum=0.5, Maximum=1.75)
    Sprite Size Max=<1.5 1.6>
    Sprite Size Min=<1 1.2>
    Sprite Size Mode=::NewEnumerator2
    UsePositionOffset=false
    Write Material Random=true
   ) {module=InitializeParticle001}
   &ShapeLocation(
    MaskSpawnGroup=false
    Offset=&MakeVector(X=0, Y=0, Z=&MakeCustomFloatFromBool(Bool=$User.bIsFrontEndPreview, True Float=-12, False Float=0))
    Offset Mode=::NewEnumerator0
    Override Seed=false
    Rotation Mode=::NewEnumerator1
    Sphere Radius=60
    Transform Order:ENiagara_TransformOrder=::NewEnumerator1
    UseOwnerScale=true
   )
   /Niagara/Modules/Math/PlaceParticlesOnDepthBuffer_GPU(Module.Use Max Z Depth=false)
   /Niagara/Modules/Update/Utility/QueryGBuffer(Buffer To Query:ENiagaraGBufferList=::NewEnumerator6)
   &KillParticles(Kill Particles=&SetBoolByFloatComparison(
     A=&MakeFloatFromLinearColor(LinearColor=$Particles.Color, Channel:ENiagaraLinearColor_Channels=::NewEnumerator0)
     B=0
     Comparison Type=::NewEnumerator5
    )) {module=KillParticles001}
   /SetVariables(Particles.Color=<rgba 60 8.0591 0 1>) {isSetParameters=true}
   &CurlNoiseForce(
    Bias Noise=false
    Noise Frequency=10
    Noise Strength=&RandomRangeFloat(Minimum=10, Maximum=14)
    Pan Noise=true
    Pan Noise Field=<0.2016 0 0.5008>
   ) {module=CurlNoiseForce001}
   &AddVelocity(
    Rotation Coordinate Space=::World
    Velocity=&RandomRangeVector(Minimum=<0 40 80>, Maximum=<1 50 120>)
    Velocity Speed Scale=0.75
   )
   &ApplyInitialForces()
   /SetVariables(Particles.EmbersSubImageIndex:NiagaraFloat=&UniformRangedFloat(Minimum=0, Maximum=36)) {isSetParameters=true}
   /SetVariables() {isSetParameters=true}
   /SetVariables(Particles.EmberSize:Vector2f=&Vector2DFromFloat(Value=&UniformRangedFloat(Minimum=0.9238, Maximum=2.4321))) {isSetParameters=true}
  ]
  particleUpdateScript [
   &ParticleState()
   &SpriteRotationRate(Rotation Rate=&Multiply_Float(
     A=&SelectFloatFromArray(
      Array Sampling Mode=::NewEnumerator0
      Direct Array Index=?
      Float Selection Array=DataInterface() {propertyValues {FloatData [-300,300]}}
      Random Seed=0
     )
     B=&RandomRangeFloat(Minimum=0.5, Maximum=1)
    ))
   &Drag(Drag=&RandomRangeFloat(Minimum=4, Maximum=5), Use Linear Drag=true, Use Rotational Drag=false)
   &GravityForce(Gravity=&RandomRangeVector(Minimum=<0 0 -8.2421>, Maximum=<1 1 -27.4089>))
   /SetVariables(Particles.EmberSize:Vector2f=&Multiply_Vector2(
     A=$Particles.Initial.EmberSize
     B=&Vector2DFromFloat(Value=&FloatFromCurve(FloatCurve=DataInterface() {propertyValues {Curve {keys [<at 0.5 1>,<at 1 0.5>]}}}))
    )) {isSetParameters=true}
   &CurlNoiseForce(
    Noise Frequency=60
    Noise Strength=&RandomRangeFloat(Minimum=30.0984, Maximum=49.8356)
    Pan Noise=true
    Pan Noise Field=<0.1 0.3 0>
   )
   &ScaleSpriteSize(Uniform Curve Sprite Scale=DataInterface() {propertyValues {Curve {keys [
        <at 0 0 RCIM_Cubic>
        {arriveTangent=0.9286, interpMode=RCIM_Cubic, leaveTangent=0.9286, time=0.3, value=0.95}
        {arriveTangent=-1.4286, interpMode=RCIM_Cubic, leaveTangent=-1.4286, time=0.7, value=0.65}
        <at 1 -0.05 RCIM_Cubic>
       ]}}})
   &SolveForcesAndVelocity(Clamp Velocity=false, Limit Acceleration=false)
   &ScaleSpriteSizeBySpeed(Max Scale Factor=<1 1>, Min Scale Factor=<0 0>, Velocity Threshold=20)
   &KillParticles(Kill Particles=&SetBoolByFloatComparison(A=&VectorLength(Vector=$Particles.Velocity), B=1, Comparison Type=::NewEnumerator6))
   &KillParticlesInVolume(Invert Volume=true, Sphere Radius=60)
  ]
  rendererClasses [&NiagaraSpriteRendererProperties]
  renderers [
   {rendererClass=&NiagaraSpriteRendererProperties}
   {rendererClass=&NiagaraSpriteRendererProperties, rendererIndex=1}
  ]
 }
 FlawlessVictory_Text {
  emitterUpdateScript [
   &EmitterState()
   &SpawnBurst_Instantaneous(Use Spawn Probability=false, UseLoopCountLimit=false)
  ]
  particleSpawnScript [
   &InitializeParticle(
    Color=<rgba 1 0.3564 0.0176 1>
    Lifetime=4
    Sprite Size=<120 30>
    Sprite Size Mode=::NewEnumerator4
    Sprite UV Mode=::NewEnumerator4
    UsePositionOffset=false
    Write Material Random=true
   )
   &ShapeLocation(
    MaskSpawnGroup=false
    Offset=&MakeVector(X=0, Y=0, Z=&MakeCustomFloatFromBool(Bool=$User.bIsFrontEndPreview, True Float=-12, False Float=0))
    Offset Coordinate Space=::Local
    Offset Mode=::NewEnumerator0
    Override Seed=false
    Sphere Radius=0.01
    UseOwnerScale=true
   )
  ]
  particleUpdateScript [
   &ParticleState()
   /SetVariables(Particles.Position=&AddVectorToPosition(Position=$Particles.ShapeLocation.ShapePosition, Vector=&MakeVector(X=0, Y=0, Z=0))) {isSetParameters=true}
   &ScaleColor(
    Scale Alpha=&FloatFromCurve(FloatCurve=DataInterface() {propertyValues {Curve {keys [
         <at 0 0 RCIM_Cubic>
         {arriveTangent=1.5, leaveTangent=1.5, time=0.2, value=1}
         {arriveTangent=-1.25, interpMode=RCIM_Cubic, leaveTangent=-1.25, time=0.8, value=1}
         {interpMode=RCIM_Cubic, leaveTangent=-0.0556, time=1, value=0}
        ]}}})
    Scale RGB=&Multiply_VectorByFloat(
     Vector=<1 1 1>
     Float=&FloatFromCurve(FloatCurve=DataInterface() {propertyValues {Curve {keys [
          {arriveTangent=-0.0114, interpMode=RCIM_Cubic, leaveTangent=-0.0114, tangentMode=RCTM_User, time=0, value=1}
          {interpMode=RCIM_Cubic, leaveTangent=-0.0556, time=1, value=1}
         ]}}})
    )
    ScaleA=true
    ScaleRGB=false
   )
  ]
  rendererClasses [&NiagaraSpriteRendererProperties]
  renderers [{rendererClass=&NiagaraSpriteRendererProperties}]
 }
 Spawn_Embers {
  simTarget=GPUComputeSim
  emitterSpawnScript [
   &InitializeAudioPlayer(Audio Player=DataInterface() {propertyValues {
      Attenuation=null
      Concurrency=null
      MaxPlaysPerTick=10
      SoundToPlay=/BRCosmetics/Sounds/Emotes/GarlicWhisk_Clobber/MSS_Emote_Foley_GarlicWhisk_Clobber_Flame
      bAllowLoopingOneShotSounds=false
      bLimitPlaysPerTick=true
      bOnlyActiveDuringGameplay=false
      bStopWhenComponentIsDestroyed=true
      ConfigurationUserParameter {parameter {name=None, typeDefHandle {registeredTypeIndex=1}, varData []}}
      ParameterNames []
     }})
   /Niagara/Modules/Audio/PlayAudio()
  ]
  emitterUpdateScript [
   &EmitterState()
   &SpawnRate(
    SpawnGroup=1
    SpawnRate=&FloatFromCurve(
     FloatCurve=DataInterface() {propertyValues {Curve {keys [<at 0 0 RCIM_Constant>,<at 0.63 1 RCIM_Constant>,<at 0.7 0>]}}}
     CurveIndex=$Emitter.NormalizedLoopAge
     Scale Curve=1000
    )
    Use Spawn Probability=false
   )
   &SpawnBurst_Instantaneous(Spawn Count=100, Spawn Time=0.1, Use Spawn Probability=false, UseLoopCountLimit=false)
   &SpawnBurst_Instantaneous(Spawn Count=100, Spawn Group=1, Spawn Time=3.625, Use Spawn Probability=false, UseLoopCountLimit=false) {enabled=false, module=SpawnBurst_Instantaneous001}
   &PlayAudio(Play Audio Trigger:NiagaraBool=&SetBoolByFloatComparison(A=$Emitter.Age, B=3.2)) {module=PlayAudio001}
  ]
  particleSpawnScript [
   &InitializeParticle(
    Color=<rgba 1.4 0.7736 0.0185 1>
    Lifetime Max=&FloatFromCurve(
     FloatCurve=DataInterface() {propertyValues {Curve {keys [<at 0 0.4 RCIM_Constant>,<at 0.5 0.1 RCIM_Constant>]}}}
     CurveIndex=$Emitter.NormalizedLoopAge
    )
    Lifetime Min:NiagaraFloat=&FloatFromCurve(
     FloatCurve=DataInterface() {propertyValues {Curve {keys [<at 0 0.05 RCIM_Constant>,<at 0.5 0.025 RCIM_Constant>]}}}
     CurveIndex=$Emitter.NormalizedLoopAge
    )
    Lifetime Mode=::NewEnumerator1
    Sprite Size Mode=::NewEnumerator1
    Sprite UV Mode=::NewEnumerator1
    Uniform Sprite Size Max=7
    Uniform Sprite Size Min=6
    UsePositionOffset=false
    Write Material Random=true
   )
   &ShapeLocation(
    MaskSpawnGroup=false
    Offset=&MakeVector(
     X=0
     Y=0
     Z=&Add_Float(
      A=&FloatFromCurve(
       FloatCurve=DataInterface() {propertyValues {Curve {keys [<at 0 4.5 RCIM_Constant>,<at 0.6 25 RCIM_Constant>,<at 0.63 25>,<at 0.69 4.27>]}}}
       CurveIndex=$Emitter.NormalizedLoopAge
      )
      B=&MakeCustomFloatFromBool(Bool=$User.bIsFrontEndPreview, True Float=-12, False Float=0)
     )
    )
    Offset Coordinate Space=::Local
    Offset Mode=::NewEnumerator0
    Override Seed=false
    Sphere Radius=1
    UseOwnerScale=true
   )
   /SetVariables(Particles.UnrealOffset:Vector2f=&MakeVector2D(
     X=&FloatFromCurve(
      FloatCurve=DataInterface() {propertyValues {Curve {keys [<at 0 8.75>,<at 1 -7.75>]}}}
      CurveIndex=&ReturnNormalizedExecIndex()
     )
     Y=0.5
    )) {isSetParameters=true}
   &AddVelocity(
    Velocity=&RandomRangeVector(Minimum=<0 0 100>, Maximum=<0 0 130>)
    Velocity Speed Scale=&FloatFromCurve(
     FloatCurve=DataInterface() {propertyValues {Curve {keys [<at 0 1 RCIM_Constant>,<at 0.5 0.1 RCIM_Constant>]}}}
     CurveIndex=$Emitter.NormalizedLoopAge
    )
   )
   &DynamicMaterialParameters(
    Index 0 Param 1=&FloatFromCurve(FloatCurve=DataInterface() {propertyValues {Curve {keys [
         {interpMode=RCIM_Cubic, leaveTangent=1.9897, tangentMode=RCTM_Break, time=0.4, value=0}
         {arriveTangent=2.0633, tangentMode=RCTM_User, time=1, value=0.5}
        ]}}})
    Index 0 Param 2=&RandomRangeFloat(Minimum=0, Maximum=1)
    Index 0 Param 3=&RandomRangeFloat(Minimum=0, Maximum=1)
    Module.Index0Float0Write=true
    Module.Index0Float1Write=true
    Module.Index0Float2Write=true
    Module.Index0Float3Write=true
   ) {module=DynamicMaterialParameters001}
  ]
  particleUpdateScript [
   &CurlNoiseForce(
    Bias Noise=false
    Noise Frequency=110
    Noise Strength=&RandomRangeFloat(Minimum=60, Maximum=110)
    Pan Noise=false
   )
   &Drag(Drag=4, Use Linear Drag=true, Use Rotational Drag=true)
   &SolveForcesAndVelocity(Clamp Velocity=false, Limit Acceleration=false)
   &ParticleState()
   &ScaleSpriteSizeBySpeed(Max Scale Factor=<1.2 2>, Min Scale Factor=<1 1.4>)
   &ScaleSpriteSize(Uniform Curve Sprite Scale=DataInterface() {propertyValues {Curve {keys [{tangentMode=RCTM_User, time=0, value=0}, <at 0.1 1>, <at 0.6 1>]}}})
   &ScaleColor(
    Scale Alpha=&FloatFromCurve(FloatCurve=DataInterface() {propertyValues {Curve {keys [
         {arriveTangent=0.1157, interpMode=RCIM_Cubic, leaveTangent=0.1157, tangentMode=RCTM_User, time=0, value=0}
         {leaveTangent=-1.4789, time=0.1, value=1}
         {arriveTangent=-1.25, interpMode=RCIM_Cubic, leaveTangent=-1.25, tangentMode=RCTM_User, time=0.8, value=1}
         {arriveTangent=0.7576, tangentMode=RCTM_User, time=1, value=0}
        ]}}})
    ScaleA=true
    ScaleRGB=true
   )
   &DynamicMaterialParameters(
    Index 0 Param 1=&FloatFromCurve(FloatCurve=DataInterface() {propertyValues {Curve {keys [
         {interpMode=RCIM_Cubic, leaveTangent=1.9897, tangentMode=RCTM_Break, time=0.4, value=0}
         {arriveTangent=-0.2983, tangentMode=RCTM_User, time=1, value=0.7}
        ]}}})
    Index 0 Param 2=&RandomRangeFloat(Minimum=0.05, Maximum=0.1)
    Index 0 Param 3=&RandomRangeFloat(Minimum=0.2, Maximum=0.25)
    Module.Index0Float0Write=true
    Module.Index0Float1Write=false
    Module.Index0Float2Write=true
    Module.Index0Float3Write=true
   )
  ]
  rendererClasses [&NiagaraSpriteRendererProperties]
  renderers [{rendererClass=&NiagaraSpriteRendererProperties}]
 }
 Spawn_Flare {
  emitterUpdateScript [
   &EmitterState()
   &SpawnBurst_Instantaneous(Use Spawn Probability=false, UseLoopCountLimit=false)
  ]
  particleSpawnScript [
   &InitializeParticle(
    Color=<rgba 1.5 0.207 0 1.5>
    Lifetime=0.25
    Position Offset=<0 0 5>
    Sprite Size=<300 75>
    Sprite Size Mode=::NewEnumerator4
    UsePositionOffset=true
    Write Material Random=true
   )
   &CameraOffset(Camera Offset Amount=30)
  ]
  particleUpdateScript [
   &ParticleState()
   &ScaleSpriteSize(Uniform Curve Sprite Scale=DataInterface() {propertyValues {Curve {keys [
        {interpMode=RCIM_Cubic, leaveTangent=0.5485, tangentMode=RCTM_Break, time=0, value=0}
        {arriveTangent=1.0163, tangentMode=RCTM_User, time=0.8, value=0.9}
       ]}}})
   &ScaleColor(
    Scale Alpha=&FloatFromCurve(FloatCurve=DataInterface() {propertyValues {Curve {keys [
         <at 0 0>
         {interpMode=RCIM_Cubic, leaveTangent=-1.4789, tangentMode=RCTM_Break, time=0.2, value=1}
         {arriveTangent=0.7576, tangentMode=RCTM_User, time=1, value=0}
        ]}}})
    ScaleA=true
    ScaleRGB=true
   )
  ]
  rendererClasses [&NiagaraSpriteRendererProperties]
  renderers [{rendererClass=&NiagaraSpriteRendererProperties}]
 }
}
summary {
 systemName=NS_GarlicWhisk_Clobber_FlawlessVictory
 emitters [
  {emitterName=Ember, simTarget=GPUComputeSim, rendererClasses [&NiagaraSpriteRendererProperties]}
  {emitterName=FlawlessVictory_Text, rendererClasses [&NiagaraSpriteRendererProperties]}
  {emitterName=Spawn_Flare, rendererClasses [&NiagaraSpriteRendererProperties]}
  {emitterName=Spawn_Embers, simTarget=GPUComputeSim, rendererClasses [&NiagaraSpriteRendererProperties]}
 ]
 userVariables [
  User.bIsFrontEnd {type=NiagaraBool}
  User.bIsFrontEndPreview {type=NiagaraBool}
  User.SkeletalMesh {
   type=NiagaraDataInterfaceSkeletalMesh
   defaultValue {
    struct=NiagaraExt_VariableValue_DataInterface
    value {dataInterface=&self:NiagaraDataInterfaceSkeletalMesh_0}
   }
  }
 ]
}
